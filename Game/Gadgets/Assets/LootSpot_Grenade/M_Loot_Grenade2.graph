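# Translucent lit material for the grenade loot spot effect, usable on mesh particles, Niagara meshes and Nanite.
# Emissive color times scale is optionally shaped by Fresnel and particle color, then fed to MF_FX_EmissiveAdjustments.
# A switch can pulse the Fresnel exponent with a time sine and dynamic parameter.
# Particle alpha is optionally multiplied by inverted Fresnel; both UseInverseFresnel branches take that same inverted output.
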Material /Game/Gadgets/Assets/LootSpot_Grenade/M_Loot_Grenade2
graph {
 properties {
  bUsedWithMeshParticles=true
  bUsedWithNanite=true
  bUsedWithNiagaraMeshParticles=true
  blendMode=BLEND_Translucent
  pixelDepthOffsetMode=PDOM_Legacy
  usedShadingModels=Default Lit
 }
 nodes [
  n1 = Sine(None=Multiply(A=Time()) {constB=0.5})
  n2 = StaticSwitchParameter(
   True=Multiply(A=DynamicParameter() {paramNames [FresnelExp,Param2,Param3,Param4]}.FresnelExp, B=Subtract(A=@n1) {constB=-2})
   False=ScalarParameter() {parameterName=FresnelExp}
  ) {parameterName=UseTimeForExp?}
  n3 = Multiply(
   A=VectorParameter() {parameterName=EmissiveColor, defaultValue=<rgba 0 0.3495 0.9322 1>}.RGB
   B=ScalarParameter() {defaultValue=8, parameterName=EmissiveColorScale}
  ) {constB=8}
  n4 = Multiply(A=MaterialFunctionCall(Exponent=@n2) {materialFunction=&MF_Fresnel_FX}.Fresnel, B=@n3)
  n5 = ParticleColor()
  MaterialFunctionCall(Emissive=StaticSwitchParameter(True=Multiply(A=@n4, B=@n5.RGB), False=@n3) {parameterName=UseFresnel?}) {materialFunction=&MF_FX_EmissiveAdjustments}
  n6 = MaterialFunctionCall(Exponent=@n2) {materialFunction=&MF_Fresnel_FX}
  StaticSwitchParameter(
   True=Multiply(A=StaticSwitchParameter(True=@n6.Fresnel Inverted, False=@n6.Fresnel Inverted) {parameterName=UseInverseFresnel?}, B=@n5.A)
   False=@n5.A
  ) {parameterName=UseFresnel?}
 ]
}
